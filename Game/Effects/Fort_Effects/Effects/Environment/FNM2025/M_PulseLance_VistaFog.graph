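# Translucent, unlit, two-sided fog material for the FNM2025 PulseLance vista effect.
# Color sums red-tinted panning cloud noise, weighted toward depth intersections, a masked glow term and a time-of-day tint with per-platform scaling.
# Opacity combines a UV mask with a noise-offset view-angle fade, platform-specific strength, depth fade and near-camera fade.
# A camera-distance sphere mask, shared by named reroute, adapts depth-fade distances between near and far.
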Material /Game/Effects/Fort_Effects/Effects/Environment/FNM2025/M_PulseLance_VistaFog
graph {
 properties {bAutomaticallySetUsageInEditor=false, blendMode=BLEND_Translucent, shadingModel=MSM_Unlit, twoSided=true, usedShadingModels=Unlit}
 nodes [
  n1 = ScalarParameter() {defaultValue=-0.1, parameterName=Param_22}
  n2 = Multiply(A=TextureCoordinate(), B=ScalarParameter() {defaultValue=25, parameterName=Param_16})
  n3 = Multiply(A=Time(), B=ScalarParameter() {defaultValue=0.05, parameterName=Param_18})
  n4 = Power(
   Base=TextureSample(UVs=Panner(Coordinate=@n2, Time=@n3) {speedY=1}) {texture=&AtmosphericCloudNoise01}.RGB
   Exp=ScalarParameter() {defaultValue=2, parameterName=Param_17}
  )
  n5 = ScalarParameter() {defaultValue=4047.7949, parameterName=Param_15}
  n6 = ScalarParameter() {parameterName=Param_3}
  n7 = NamedRerouteUsage()
  n8 = OneMinus(None=DepthFade(FadeDistance=LinearInterpolate(A=@n5, B=@n6, Alpha=@n7)) {fadeDistanceDefault=7500})
  n9 = SphereMask(
   A=TextureCoordinate()
   B=ScalarParameter() {defaultValue=0.5, parameterName=color size}
   Radius=ScalarParameter() {defaultValue=0.25, parameterName=Secondary Glow}
   Hardness=ScalarParameter() {parameterName=color hardness}
  )
  n10 = VectorParameter()
  n11 = Multiply(
   A=MaterialFunctionCall(
    Color_Morning=VectorParameter() {parameterName="Color_Morning (V3)", defaultValue=<rgba 1 1 1 1>}.RGB
    Color_Day=VectorParameter() {parameterName="Color_Day (V3)", defaultValue=<rgba 1.5 1.5 1.5 1>}.RGB
    Color_Evening=VectorParameter() {parameterName="Color_Evening (V3)", defaultValue=<rgba 1 1 1 1>}.RGB
    Color_Night=VectorParameter() {parameterName="Color_Night (V3)", defaultValue=<rgba 1 1 1 1>}.RGB
   ) {materialFunction=&Color_Time_of_Day}.Result
   B=VectorParameter() {parameterName=Param_1, defaultValue=<rgba 1 0.0048 0 1>}.RGB
  )
  n12 = ScalarParameter() {defaultValue=1, parameterName=Param_23}
  n13 = ScalarParameter() {defaultValue=1.25, parameterName=Param_24}
  n14 = MaterialFunctionCall!("Xb1,PS4"=@n12, "PC-High&&Epic,XSX,PS5"=@n12, "PC-Perf,Android"=@n13, Switch=@n13, "PC-Low(NotPerf)&&Medium"=@n12) {materialFunction=/Game/Effects/Niagara/MasterMaterials/Functions/Common/PlatformSwitch_3D}
  Add(
   A=Multiply(A=LinearInterpolate(A=@n1, B=@n4, Alpha=@n8), B=VectorParameter() {parameterName=Param_2, defaultValue=<rgba 2 0 0.0317 1>}.RGB)
   B=Add(A=Multiply(A=@n9, B=@n10.RGB), B=Multiply(A=@n11, B=@n14.Result))
  )
  n15 = CameraVectorWS()
  n16 = Multiply(
   A=TextureCoordinate()
   B=AppendVector(A=ScalarParameter() {defaultValue=1.25, parameterName=Param_10}, B=ScalarParameter() {defaultValue=1.5, parameterName=Param_11})
  )
  n17 = Multiply(A=Time(), B=ScalarParameter() {defaultValue=0.01, parameterName=Param_9})
  n18 = Multiply(
   A=TextureSample(UVs=Panner(Coordinate=@n16, Time=@n17) {speedX=1}) {texture=&AtmosphericCloudNoise01}.G
   B=ScalarParameter() {parameterName=Noise Distort}
  )
  n19 = TwoSidedSign()
  n20 = VertexNormalWS()
  n21 = Saturate(None=DotProduct(A=Add(A=@n15, B=@n18), B=Multiply(A=@n19, B=@n20)))
  n22 = ScalarParameter() {defaultValue=1, parameterName=Camera Start Fade}
  n23 = ScalarParameter() {defaultValue=0.596, parameterName=Radius}
  n24 = ScalarParameter() {parameterName=Hardness}
  n25 = Multiply(
   A=SphereMask(
    A=TextureCoordinate()
    B=ScalarParameter() {defaultValue=0.5}
    Radius=ScalarParameter() {defaultValue=8, parameterName=Param_1}
    Hardness=ScalarParameter() {parameterName=Param_2}
   )
   B=Power(Base=SphereMask(A=@n21, B=@n22, Radius=@n23, Hardness=@n24), Exp=ScalarParameter() {defaultValue=2, parameterName=Camera Angle Fade Power})
  )
  n26 = ScalarParameter() {defaultValue=0.35, parameterName=Opacity}
  n27 = ScalarParameter() {defaultValue=0.35, parameterName=Opacity Low}
  n28 = MaterialFunctionCall("Xb1,PS4"=@n26, "PC-High&&Epic,XSX,PS5"=@n26, "PC-Perf,Android"=@n27, Switch=@n26, "PC-Low(NotPerf)&&Medium"=@n27) {materialFunction=&PlatformSwitch_3D}
  n29 = ScalarParameter() {defaultValue=3500, parameterName=Depth Fade Dist Far}
  n30 = ScalarParameter() {defaultValue=150, parameterName=Depth Fade Distance Close}
  n31 = NamedRerouteUsage()
  Multiply(
   A=MaterialFunctionCall(Distance=ScalarParameter() {defaultValue=1500, parameterName=Near Camera Fade}) {materialFunction=&Near_Camera_Fade}.Result
   B=DepthFade(Opacity=Multiply(A=@n25, B=@n28.Result), FadeDistance=LinearInterpolate(A=@n29, B=@n30, Alpha=@n31)) {fadeDistanceDefault=7500}
  )
  n32 = SphereMask(
   A=WorldPosition().XYZ
   B=CameraPositionWS()
   Radius=ScalarParameter() {defaultValue=5500, parameterName=Param_19}
   Hardness=ScalarParameter() {defaultValue=0.35, parameterName=Param_20}
  )
  NamedRerouteDeclaration(None=@n32) {name=Camera Fade, nodeColor=<rgba 0.4297 0 1 1>}
  OneMinus(None=@n32)
 ]
}
